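# Subsurface wax material for posed DesertCity characters, usable on skeletal, Nanite, spline and deformed meshes.
# A fake specular highlight from the sky light's half vector, tinted by time of day and masked by packed-texture roughness, is added to base color.
# The packed texture also drives specular and metallic, while inverted Fresnel shapes subsurface opacity.
# A fake-light and gradient emissive branch is present but unconnected to the emissive output.
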Material /Hera/Sets/DesertCity/Meshes/Posed_Characters/M_WaxPosedCharacters
graph {
 properties {
  bTangentSpaceNormal=false
  bUseMaterialAttributes=true
  bUsedWithMeshDeformer=true
  bUsedWithMorphTargets=true
  bUsedWithNanite=true
  bUsedWithSkeletalMesh=true
  bUsedWithSplineMeshes=true
  bUsedWithStaticLighting=true
  pixelDepthOffsetMode=PDOM_Legacy
  shadingModel=MSM_Subsurface
  usedShadingModels=Subsurface
 }
 nodes [
  Constant()
  Constant() {r=1}
  Constant() {r=1}
  Constant()
  n1 = NamedRerouteUsage()
  DotProduct(A=@n1, B=CameraVectorWS())
  n2 = Multiply(A=DotProduct(A=NamedRerouteUsage(), B=Constant3Vector() {constant=<rgba 0 0 -1 0>})) {constB=0.5}
  n3 = Power(Base=Saturate(None=Add(A=@n2) {constB=0.5})) {constExponent=3}
  n4 = VectorParameter() {parameterName=Fake Light Color, defaultValue=<rgba 0.9514 0.1503 1 1>}
  n5 = ObjectLocalBounds()
  n6 = Saturate(None=Divide(A=LocalPosition().Z, B=ComponentMask(None=@n5.Extents) {b=true}))
  n7 = Multiply(
   A=Power(Base=OneMinus(None=@n6), Exp=ScalarParameter() {defaultValue=4, parameterName=Gradient Light Power})
   B=ScalarParameter() {defaultValue=1, parameterName=Gradient Light Intensity}
  )
  n8 = VectorParameter() {parameterName=Gradient Light Color, defaultValue=<rgba 0.7128 0.0294 1 1>}
  n9 = Multiply(A=Add(A=Multiply(A=@n3, B=@n4.RGB), B=Multiply(A=@n7, B=@n8.RGB)), B=ScalarParameter() {defaultValue=10, parameterName=Fake Light Intensity})
  n10 = NamedRerouteUsage()
  n11 = Multiply(A=@n9, B=@n10)
  n12 = Multiply(A=@n9) {constB=0.1}
  n13 = VectorParameter() {parameterName=Emission Color, defaultValue=<rgba 0.0168 0.0269 0.0573 1>}
  EyeAdaptationInverse(LightValueInput=Add(A=Add(A=@n11, B=@n12), B=Multiply(A=@n10, B=@n13.RGB)), AlphaInput=Constant() {r=0.75})
  n14 = MaterialFunctionCall(Color_Day=VectorParameter() {parameterName=Color day, defaultValue=<rgba 1 0.2904 0 1>}.RGB) {materialFunction=&Color_Time_of_Day}
  n15 = Add(A=SkyAtmosphereLightDirection(), B=CameraVectorWS())
  n16 = Power(
   Base=DotProduct(A=Normalize(VectorInput=@n15), B=NamedRerouteUsage())
   Exp=ScalarParameter() {defaultValue=6, parameterName=Fake Specular Contrast}
  )
  n17 = TextureCoordinate()
  n18 = TextureSampleParameter2D(UVs=@n17) {
   parameterName=PackedTexture
   samplerSource=SSM_Clamp_WorldGroupSettings
   samplerType=SAMPLERTYPE_LinearColor
   texture=/Engine/EngineMaterials/BaseFlattenLinearColor
  }
  n19 = Multiply(A=@n18.B, B=ScalarParameter() {defaultValue=1, parameterName=Roughness Multiply})
  n20 = Multiply(A=@n19, B=ScalarParameter() {defaultValue=1, parameterName=Fake Contrast Intensity}) {constB=2}
  n21 = Saturate(None=Multiply(A=LinearInterpolate(A=@n14.Result) {constAlpha=0.3}, B=Multiply(A=@n16, B=@n20)))
  Multiply(A=@n21) {constB=0.5}
  PixelNormalWS()
  n22 = TextureSampleParameter2D(UVs=@n17) {
   parameterName=AmbientOcclusionTexture
   samplerType=SAMPLERTYPE_LinearGrayscale
   texture=/Engine/EngineMaterials/BaseFlattenGrayscaleMap.BaseFlattenGrayScaleMap
  }
  n23 = StaticSwitchParameter(True=@n22.RGB, False=ScalarParameter() {defaultValue=1, parameterName=AmbientOcclusionConst}) {parameterName=UseAmbientOcclusion}
  n24 = TextureSampleParameter2D(UVs=@n17) {
   parameterName=OpacityTexture
   samplerType=SAMPLERTYPE_LinearGrayscale
   texture=&BaseFlattenGrayScaleMap
  }
  n25 = ScalarParameter() {defaultValue=1, parameterName=OpacityConst}
  n26 = TextureSampleParameter2D(UVs=@n17) {
   parameterName=OpacityMaskTexture
   samplerType=SAMPLERTYPE_LinearGrayscale
   texture=&BaseFlattenGrayScaleMap
  }
  n27 = ScalarParameter() {defaultValue=1, parameterName=OpacityMaskConst}
  n28 = TextureSampleParameter2D(UVs=@n17) {
   parameterName=EmissiveHDRTexture
   samplerSource=SSM_Clamp_WorldGroupSettings
   samplerType=SAMPLERTYPE_LinearColor
   texture=/Engine/Functions/Engine_MaterialFunctions02/PivotPainter2/Black_1x1_EXR_Texture
  }
  n29 = TextureSampleParameter2D(UVs=@n17) {parameterName=EmissiveColorTexture, samplerSource=SSM_Clamp_WorldGroupSettings, texture=/Engine/EngineMaterials/BaseFlattenEmissiveMap}
  n30 = StaticSwitchParameter(True=@n29.RGB, False=VectorParameter() {parameterName=EmissiveColorConst}.RGB) {parameterName=UseEmissiveColor}
  n31 = ScalarParameter() {defaultValue=1, parameterName=EmissiveScale}
  SetMaterialAttributes(
   MaterialAttributes=SetMaterialAttributes!(
    MaterialAttributes=SetMaterialAttributes!(MaterialAttributes, Ambient Occlusion=@n23) {attributeSetTypes [#1]}
    Opacity=StaticSwitchParameter(True=@n24.RGB, False=@n25) {parameterName=UseOpacity}
    Opacity Mask=StaticSwitchParameter(True=@n26.RGB, False=@n27) {parameterName=UseOpacityMask}
   ) {attributeSetTypes [#2, #3]}
   Emissive Color=StaticSwitchParameter(True=@n28.RGB, False=Multiply(A=@n30, B=@n31)) {parameterName=UseEmissiveHDR}
  ) {attributeSetTypes [#4]}
  n32 = LinearInterpolate(B=@n18.G, Alpha=ScalarParameter() {parameterName=Metallic Noise, sliderMax=1})
  n33 = Multiply(A=Add(A=@n18.R) {constB=0.2}, B=ScalarParameter() {defaultValue=2, parameterName=Spec Multiply})
  n34 = TextureSampleParameter2D(UVs=@n17) {
   parameterName=NormalTexture
   samplerSource=SSM_Clamp_WorldGroupSettings
   samplerType=SAMPLERTYPE_Normal
   texture=/Engine/EngineMaterials/BaseFlattenNormalMap
  }
  n35 = Transform(None=@n34.RGB)
  n36 = TextureSampleParameter2D(UVs=@n17) {parameterName=BaseColorTexture, texture=/Engine/EngineMaterials/BaseFlattenDiffuseMap}
  n37 = Add(A=@n21, B=LinearInterpolate(A=@n36.RGB) {constAlpha=0.2, constB=0.35})
  n38 = VectorParameter() {parameterName=Color Multiply, defaultValue=<rgba 0.2188 0.1735 0.0891 1>}
  n39 = Multiply(A=@n38.RGB, B=Saturate(None=VectorParameter() {parameterName=Subsurface Color, defaultValue=<rgba 0.8958 0.671 0.4293 1>}.RGB))
  n40 = ScalarParameter() {defaultValue=0.5, parameterName=SS Constant Color Bias}
  n41 = ScalarParameter() {defaultValue=0.4871, parameterName=Min Opacity}
  n42 = Fresnel(ExponentIn=ScalarParameter() {defaultValue=0.2962, parameterName=Edges Contrast}, Normal=@n1) {baseReflectFraction=0, exponent=1}
  n43 = Saturate(None=OneMinus(None=Multiply(A=@n42, B=@n42)))
  SetMaterialAttributes!(
   MaterialAttributes=SetMaterialAttributes!(
    MaterialAttributes=SetMaterialAttributes!(MaterialAttributes, Metallic=@n32, Specular=@n33, Roughness=@n19) {attributeSetTypes [#5, #6, #7]}
    Normal=NamedRerouteDeclaration(None=@n35) {name=World Normal, nodeColor=<rgba 0.2891 1 0 1>}
   ) {attributeSetTypes [#8]}
   Base Color=Multiply(A=NamedRerouteDeclaration(None=@n37) {name=Base Color, nodeColor=<rgba 0 0.8594 1 1>}, B=@n38.RGB)
   Subsurface Color=Multiply(A=@n36.RGB, B=LinearInterpolate(B=@n39, Alpha=@n40) {constA=1})
   Opacity=Saturate(None=Add(A=@n41, B=@n43))
   Emissive Color=Reroute()
  ) {attributeSetTypes [
    #9
    #10
    #2
    #4
   ]}
  n44 = TextureCoordinate() {coordinateIndex=2}
  n45 = TextureCoordinate() {coordinateIndex=5}
  n46 = StaticSwitchParameter(True=TextureCoordinate() {coordinateIndex=6}, False=TextureCoordinate()) {parameterName=UseUV6}
  n47 = StaticSwitchParameter(
   True=TextureCoordinate() {coordinateIndex=3}
   False=StaticSwitchParameter(True=TextureCoordinate() {coordinateIndex=4}, False=StaticSwitchParameter(True=@n45, False=@n46) {parameterName=UseUV5}) {parameterName=UseUV4}
  ) {parameterName=UseUV3}
  StaticSwitchParameter(
   True=StaticSwitchParameter(True=TextureCoordinate() {coordinateIndex=1}, False=StaticSwitchParameter(True=@n44, False=@n47) {parameterName=UseUV2}) {parameterName=UseUV1}
   False=@n17
  ) {parameterName=UseCustomUV}
  VertexNormalWS()
 ]
}
